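; Battle Pass quest ability that tracks "Wild Air Time", seconds spent falling under Hop Flopper or Shock Grenade effects.
; Adding the fall-immunity tag starts a looping one-second timer, and removing it clears the timer.
; Each tick checks the pawn's tags and increments a counter when falling with either effect.
; At five accrued seconds it sends a complex custom quest stat event and resets, so shorter stretches carry over.
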(event Ability|EventActivateAbility
  (bind _t_fallimmunity (Variables|Default|GetT_FallImmunity))
  ;[1 1 1] Wild Air Time
  (|CallParentFunction)
  (Utilities|FlowControl|Sequence
    (:then_0
      (Ability|Tasks|WaitGameplayTagAdd _t_fallimmunity
        (:Added
          (bind _returnvalue (Utilities|Time|SetTimerbyEvent (delegate Custom|Watch%20for%20Wild%20AirTime) 1.0 true))
          (Variables|Default|SetTimer%20-%20AirTime _returnvalue))) ; added tag: falling immunity
    )
    (:then_1
      (Ability|Tasks|WaitGameplayTagRemove _t_fallimmunity
        (:Removed
          (Utilities|Time|ClearandInvalidateTimerbyHandle (Variables|Default|GetTimer%20-%20AirTime)))) ; removed tag: falling immunity
    )))

(event Custom|Watch%20for%20Wild%20AirTime
  (bind _owningplayerpawn (Variables|GAVars|Data|GetOwningPlayerPawn))
  ;[1 1 1] Wild Air Time
  (Utilities|FlowControl|Sequence
    (:then_0
      (bind _bdidactorhavetags (Class|AthenaQuestFunctionLibrary|AthenaCheckActorHasTags _owningplayerpawn "(GameplayTags=((TagName=\"Status.LowGravity.LowGravRock\")))")) ; Hop Flopper
    )
    (:then_1
      (bind _bdidactorhavetags_2 (Class|AthenaQuestFunctionLibrary|AthenaCheckActorHasTags _owningplayerpawn "(GameplayTags=((TagName=\"Granted.Athena.Item.ShockGrenade.Knocked\")))")) ; Shock Grenade
    )
    (:then_2
      (bind _bdidactorhavetags_3 (Class|AthenaQuestFunctionLibrary|AthenaCheckActorHasTags _owningplayerpawn "(GameplayTags=((TagName=\"MovementMode.Falling\")))")) ; In the Air
    )
    (:then_3
      (if (and (or _bdidactorhavetags _bdidactorhavetags_2) _bdidactorhavetags_3)
        (Math|Integer|IncrementInt (Variables|Default|Gettime%20%28seconds%29)
          (:%20%20
            (if (>= (Variables|Default|Gettime%20%28seconds%29) 5)
              (Quest|SendComplexCustomStatEvent (Variables|GAVars|Data|GetOwningPlayerQuestManager) 0 (Variables|Default|GetT_Quest) "(GameplayTags=)" (Variables|Default|Gettime%20%28seconds%29))
              (Development|PrintString (Utilities|String|Append "LogFortQuest [Wild Air Time] - sending Complex Custom, seconds accrued = " (Utilities|String|ToString(Integer) (Variables|Default|Gettime%20%28seconds%29))) false) ; debug - when quest event sent
              (Variables|Default|Settime%20%28seconds%29 0) ; reset accrued airtime counter so we only \nsend the amount accrued since the last\ntime we sent the event
            ))) ; +1 SECOND WILD AIRTIME
      ))))

(event Ability|EventOnEndAbility
  (params (bWasCancelled :type Boolean)))

(event Ability|EventFailedToActivatePassiveAbility
  (params (ActorInfo :type Gameplay%20Ability%20Actor%20Info%20Structure)))
