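; Event-triggered Battle Pass Season 16 ability for the Wild Air Time quest, measuring player airtime.
; On the FallDamageImmunity cue, it waits for Shockwave knocked-tag removal and contains a gated, delayed counter with debug logging.
; Check In Air adds a fixed interval while the player's vehicle is airborne, resetting otherwise.
; At the required time it sends the quest stat and clears the timer, whose setup is absent here.
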(event Ability|EventActivateAbilityFromEvent
  (params (EventData :type Gameplay%20Event%20Data%20Structure%20%28by%20ref%29))
  (|CallParentFunction EventData)
  (switch Utilities|FlowControl|Switch|SwitchonGameplayTag (Utilities|Struct|BreakGameplayEventData EventData)
    (:GameplayCue.Athena.FallDamageImmunity
      (Utilities|FlowControl|Sequence
        (:then_0
          _)
        (:then_1
          _)
        (:then_2
          (Ability|Tasks|WaitGameplayTagRemove "(TagName=\"Granted.Athena.Item.ShockGrenade.Knocked\")" (Variables|GAVars|Data|GetOwningPlayerPawn)
            (:Removed
              _))))
      (Utilities|FlowControl|Gate false
        (:Exit
          (Development|PrintString (Utilities|String|Append "LogFortQuest: Wild Airtime: Shockwave - " (Utilities|String|ToString(Integer) (Variables|Default|Getcounter))) true false)
          (Utilities|FlowControl|Delay 0.2)
          (Math|Integer|IncrementInt (Variables|Default|Getcounter)
            (:%20%20)))))))

(event Custom|Check%20In%20Air
  (Utilities|Casting|CastToFortAthenaVehicle (Vehicle|GetVehicleInterface (Variables|GAVars|Data|GetOwningPlayerPawn))
    (:then
      (if (Vehicle|VehicleIsinAir (Vehicle|GetVehicleInterface (Variables|GAVars|Data|GetOwningPlayerPawn)))
        ;[1 1 1] Add the time spent in air to the total elapsed time in air (as calculated by the time interval)
        (bind _output_get (Variables|Default|SetElapsedTimeInAir (+ (Variables|Default|GetExecutionTimeInterval) (Variables|Default|GetElapsedTimeInAir))))
        (if (>= _output_get (Variables|Default|GetRequiredTime))
          (Quest|SendComplexCustomStatEvent (Variables|GAVars|Data|GetOwningPlayerQuestManager) 0 (Variables|Default|GetT_Quest) "(GameplayTags=)" 1)
          (Utilities|Time|ClearandInvalidateTimerbyHandle (Variables|Default|GetTimer%20Handle)))
        (else
          (Variables|Default|SetElapsedTimeInAir 0.0))))
    (:CastFailed
      _))
  (Ability|EndAbility))
